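; BurgleBoss companion part modifier that grows a reactive gold mask with player kills when the Stage0 parts variant is active.
; In game, it watches the AthenaKills stat, maps the clamped stage to per-stage material progress, noise tiling and emissive values, and timestamps changes for shader-driven transitions.
; In the frontend, Tick and delayed custom events drive a timed stage preview, gated by the customization screen tag and a specific emote montage.
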(event Custom|StartReactivity
  ;[0 0.2896 1 0.5] Frontend Preview
  (Variables|Default|SetIsReactiveRunning true)
  (CallFunction|TransitionLevel (Variables|Default|GetGM_CurrentStage))
  (CallFunction|UpdateStageCounter)
  (if (Variables|Default|GetShouldRunReactive)
    (Utilities|FlowControl|Delay 5.0))
  (CallFunction|StopReactivity))

(event EventTick
  (params (DeltaSeconds :type Float%20%28single-precision%29))
  ;[0 0.2896 1 0.5] Frontend Preview
  (|CallParentFunction DeltaSeconds)
  (if (CustomCosmeticModifier|IsinFrontEnd)
    (bind _asfort_player_companion_preview_actor (Utilities|Casting|CastToFortPlayerCompanionPreviewActor (Actor|GetOwner)))
    (Variables|Default|SetShouldRunReactive (and (!= (Animation|Montage|GetCurrentActiveMontage (Components|SkeletalMesh|GetAnimInstance (Variables|/Script/CosmeticCompanionsRuntime.FortPlayerCompanionPreviewActor|GetMesh _asfort_player_companion_preview_actor))) "/CosmeticCompanions/Assets/Quadruped/BurgleBoss/Animation/Emotes/BurgleBoss_Emote_M.BurgleBoss_Emote_M") (Variables|Default|GetIsReactive)))
    (if (!= (Variables|Default|GetShouldRunReactive) (Variables|Default|GetIsReactiveRunning))
      (if true
        (CallFunction|StartReactivity)
        (else
          (CallFunction|StopReactivity))))))

(event Custom|StopReactivity
  ;[0 0.2896 1 0.5] Frontend Preview
  (Variables|Default|SetIsReactiveRunning false)
  (CallFunction|TransitionLevel 0)
  (Variables|Default|SetGM_CurrentStage 0))

(event EventOnPostResIn
  (bind _returnvalue (CustomCharacter|GetSkeletalMeshForPartOnPlayerPawn))
  (bind _gm_endstage (Variables|Default|GetGM_EndStage))
  (bind _returnvalue_2 (+ _gm_endstage 1))
  (bind _gm_laststage (Variables|Default|GetGM_LastStage))
  ;[1 1 1] GoldMaskGrowingLogic
  (|CallParentFunction)
  (if (CustomCosmeticModifier|IsinFrontEnd)
    ;[0 0.2896 1 0.5] Frontend Preview
    (Utilities|FlowControl|Delay 2.0)
    (bind _output_get (Variables|Default|SetIsReactive (and (== (Class|FortCosmeticFunctionLibrary|GetActiveVariantForSoftItemAndChannel (CosmeticLoadout|GetCosmeticLoadout) "/CosmeticCompanions/Assets/Items/Companion_BurgleBoss.Companion_BurgleBoss" "(TagName=\"Cosmetics.Variant.Channel.Parts\")") "(TagName=\"Cosmetics.Variant.Property.Stage0\")") (not (GameplayTags|HasTag (Reactive|Frontend|TryGetPreviewContextTags self) "(TagName=\"UI.ItemPreview.SceneChanger.Customization\")")))))
    (if _output_get
      (CallFunction|StartReactivity)
      (else
        (CallFunction|StopReactivity)))
    (else
      (bind _output_get_2 (Variables|Default|SetIsReactive (== (Class|FortCosmeticFunctionLibrary|GetActiveVariantForSoftItemAndChannel (CosmeticLoadout|GetCosmeticLoadout) "/CosmeticCompanions/Assets/Items/Companion_BurgleBoss.Companion_BurgleBoss" "(TagName=\"Cosmetics.Variant.Channel.Parts\")") "(TagName=\"Cosmetics.Variant.Property.Stage0\")")))
      (if _output_get_2
        (bind (_statname _statvalue) (Cosmetics|WatchCosmeticStat (CustomCharacter|GetAssociatedPlayerPawn) "(TagName=\"StatManager.AthenaKills\")")
          (:OnStatChanged
            (Variables|Default|SetGM_LastStage (Math|Integer|SelectInt (Math|Integer|Max(Integer) _statvalue _gm_laststage) _gm_endstage (< (Math|Integer|Max(Integer) _statvalue _gm_laststage) _returnvalue_2)))
            (Utilities|FlowControl|Sequence
              (:then_0
                ;; Reactivity Per Stage look
                (CallFunction|InitializeVariabels)
                (Rendering|Material|SetScalarParameterValueonMaterials _returnvalue "GM_CurrentStage" (Utilities|Array|Get(aref) (Variables|Default|GetGM_ProgessPerStage) (Variables|Default|GetGM_LastStage)))
                (Rendering|Material|SetVectorParameterValueonMaterials _returnvalue "GM_TransitionNoise_Tiling" (Utilities|Array|Get(acopy) (Variables|Default|GetGM_NoiseTillingPerStage) (Variables|Default|GetGM_LastStage)))
                (Rendering|Material|SetScalarParameterValueonMaterials _returnvalue "GM_Transition_EmissiveIntensity" (Utilities|Array|Get(aref) (Variables|Default|GetGM_EmissiveIntensityPerStage) (Variables|Default|GetGM_LastStage))))
              (:then_1
                (if (and (> (Math|Integer|Max(Integer) _statvalue _gm_laststage) (Variables|Default|GetGM_StartStage)) (< (Math|Integer|Max(Integer) _statvalue _gm_laststage) _returnvalue_2))
                  (CallFunction|OnActualStatChange))))))
        (else
          (CallFunction|TransitionLevel 0))))))

(event Custom|OnActualStatChange
  ;[1 1 1] GoldMaskGrowingLogic
  ;;[1 1 1] This should only fire off when the stat changes - use this event!
  (Rendering|Material|SetScalarParameterValueonMaterials (CustomCharacter|GetSkeletalMeshForPartOnPlayerPawn) "GM_GameTimeFromBP" (Utilities|Time|GetGameTimeinSeconds)))

(event Custom|InitializeVariabels
  (bind _returnvalue (CustomCharacter|GetSkeletalMeshForPartOnPlayerPawn))
  ;[1 1 1] GoldMaskGrowingLogic
  ;; Initialize Variables
  (Rendering|Material|SetScalarParameterValueonMaterials _returnvalue "GM_StartAtStage" (Math|Conversions|ToFloat(Integer) (Variables|Default|GetGM_StartStage)))
  (Rendering|Material|SetScalarParameterValueonMaterials _returnvalue "GM_StopAtStage" (Math|Conversions|ToFloat(Integer) (Variables|Default|GetGM_EndStage)))
  (Rendering|Material|SetScalarParameterValueonMaterials _returnvalue "GM_Transition_EmissiveIntensity" 0.0))

(fn TransitionLevel
  (params (Stage :type Integer))
  (bind _returnvalue (CustomCharacter|GetSkeletalMeshForPartOnPlayerPawn))
  (Rendering|Material|SetScalarParameterValueonMaterials _returnvalue "GM_CurrentStage" (Utilities|Array|Get(aref) (Variables|Default|GetGM_ProgessPerStage) Stage))
  (Rendering|Material|SetScalarParameterValueonMaterials _returnvalue "GM_GameTimeFromBP" (Utilities|Time|GetGameTimeinSeconds))
  (Rendering|Material|SetVectorParameterValueonMaterials _returnvalue "GM_TransitionNoise_Tiling" (Utilities|Array|Get(acopy) (Variables|Default|GetGM_NoiseTillingPerStage) Stage))
  (Rendering|Material|SetScalarParameterValueonMaterials _returnvalue "GM_Transition_EmissiveIntensity" (Utilities|Array|Get(aref) (Variables|Default|GetGM_EmissiveIntensityPerStage) Stage)))

(fn UpdateStageCounter
  (if (>= (Variables|Default|GetGM_CurrentStage) (Variables|Default|GetGM_EndStage))
    (Variables|Default|SetGM_CurrentStage 0)
    (else
      (Math|Integer|IncrementInt (Variables|Default|GetGM_CurrentStage)))))

(fn ConstructionScript
  (bind _returnvalue (CustomCharacter|GetSkeletalMeshForPartOnPlayerPawn))
  (|CallParentFunction)
  (Rendering|Material|SetScalarParameterValueonMaterials (CustomCharacter|GetSkeletalMeshForPartOnPlayerPawn) "GM_GameTimeFromBP" -99999.0)
  (Rendering|Material|SetScalarParameterValueonMaterials _returnvalue "GM_StartAtStage" (Math|Conversions|ToFloat(Integer) (Variables|Default|GetGM_StartStage)))
  (Rendering|Material|SetScalarParameterValueonMaterials _returnvalue "GM_StopAtStage" (Math|Conversions|ToFloat(Integer) (Variables|Default|GetGM_EndStage)))
  (Rendering|Material|SetScalarParameterValueonMaterials _returnvalue "GM_Transition_EmissiveIntensity" 0.0))
