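# Eye material for the BurgleBoss quadruped cosmetic companion, adding an optional "scar eye" recolor to one eye.
# MF_animal_eyes_2025 supplies the base attributes, and only Base Color is replaced before the result passes through MF_TechArtOverrides.
# The scar color blends the diffuse red channel toward ScarEye_Color by ScarEyeInfluence.
# ScarEye drives the effect, masked to negative X using the sign of vertex-interpolated pre-skinning local position.
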Material /CosmeticCompanions/Assets/Quadruped/BurgleBoss/Materials/M_Companion_Eyes_BurgleBoss
graph {
 properties {
  bUseMaterialAttributes=true
  bUsedWithMeshDeformer=true
  bUsedWithMorphTargets=true
  bUsedWithSkeletalMesh=true
  pixelDepthOffsetMode=PDOM_Legacy
  shadingModel=MSM_FromMaterialExpression
  usedShadingModels=Default Lit
 }
 nodes [
  n1 = GetMaterialAttributes(None=MaterialFunctionCall() {materialFunction=/Game/Global/Materials/Character_Parent_2023/Eye_Parent_2023/MF_animal_eyes_2025}.Result) {attributeGetTypes [#1]}
  n2 = LinearInterpolate(
   A=TextureSampleParameter2D() {group=00_Eye_Basic, parameterName=Diffuse, sortPriority=0, texture=/Game/Characters/Shared/FNEye/Textures/T_Eye_D}.R
   B=VectorParameter() {parameterName=ScarEye_Color, defaultValue=<rgba 1 0.9174 0.4977 1>}.RGB
   Alpha=ScalarParameter() {parameterName=ScarEyeInfluence}
  ) {constA=1}
  n3 = LocalPosition() {localOrigin=InstancePreSkinning}
  n4 = Sign(None=ComponentMask(None=VertexInterpolator(VS=@n3.XYZ).PS) {r=true})
  n5 = Saturate(None=Multiply(A=OneMinus(None=@n4), B=ScalarParameter() {parameterName=ScarEye}))
  MaterialFunctionCall(In=SetMaterialAttributes!(MaterialAttributes=@n1.MaterialAttributes, Base Color=LinearInterpolate(A=@n1.Base Color, B=@n2, Alpha=@n5) {constAlpha=0}) {attributeSetTypes [#1]}) {materialFunction=&MF_TechArtOverrides}
 ]
}
