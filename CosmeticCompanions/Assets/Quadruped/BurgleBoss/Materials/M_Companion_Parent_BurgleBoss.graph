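# Parent material for the BurgleBoss quadruped companion, layering color customization and gold effects between shared CharacterParent PreFX and PostFX functions.
# Optional customization recolors base color through masked primary, secondary, tertiary and accent layers, with switchable blend colors and pattern overlay.
# A borrowed AuraCop gold mask, sampled in pre-skinned space, drives a stage-gated metallic gold blend with emissive transition.
# Tech Art mask channels define constant gold and toggleable scar-eye areas; gold normals flatten small detail.
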Material /CosmeticCompanions/Assets/Quadruped/BurgleBoss/Materials/M_Companion_Parent_BurgleBoss
graph {
 properties {
  bAutomaticallySetUsageInEditor=false
  bUseMaterialAttributes=true
  bUsedWithMeshDeformer=true
  bUsedWithMorphTargets=true
  bUsedWithSkeletalMesh=true
  bUsedWithStaticLighting=true
  shadingModel=MSM_FromMaterialExpression
  usedShadingModels=Default Lit | Subsurface | Clear Coat | Subsurface Profile | Cloth
 }
 nodes [
  n1 = MaterialFunctionCall() {materialFunction=&MF_CharacterParent_2023_PreFX}
  MaterialFunctionCall(Material=NamedRerouteUsage(), M=@n1.M, Pre Skinned Local Position=@n1.M, Distance Blend=@n1.M, Fresnel=@n1.M) {materialFunction=&MF_CharacterParent_2023_PostFX}
  n2 = TextureSampleParameter2D() {
   group=Color Customization
   parameterName=Pattern Mask
   samplerSource=SSM_Wrap_WorldGroupSettings
   samplerType=SAMPLERTYPE_Masks
   sortPriority=51
   texture=&T_BlackMask
  }
  NamedRerouteDeclaration(None=ChannelMaskParameter(None=@n2.RGB) {desc=Uses Pattern Mask for accent, group=Color Customization, parameterName=AccentMaskChannel, sortPriority=12}.R) {name=AccentMask, nodeColor=<rgba 1 0.9075 0.0088 1>}
  n3 = VectorParameter() {group=Color Customization, parameterName=Accent Color, sortPriority=36, defaultValue=<rgba 1 1 1 1>}
  NamedRerouteDeclaration(None=@n3.RGB) {name=AccentColor, nodeColor=<rgba 0 1 0.7891 1>}
  n4 = TextureSampleParameter2D() {
   group=Color Customization
   parameterName=Custom Color Mask
   samplerSource=SSM_Wrap_WorldGroupSettings
   samplerType=SAMPLERTYPE_Masks
   sortPriority=11
   texture=&T_BlackMask
  }
  n5 = ChannelMaskParameter(None=@n4.RGB) {group=Color Customization, maskChannel=Blue, parameterName=TertiaryMaskChannel, sortPriority=14, defaultValue=<rgba 0 0 1 0>}
  NamedRerouteDeclaration(None=@n5.R) {name=TertiaryMask, nodeColor=<rgba 0 0.2243 1 1>}
  NamedRerouteDeclaration(None=StaticBoolParameter() {group=Color Customization, parameterName=UseAdvancedColorBlend, sortPriority=20}) {name=bAdvancedColorBlend, nodeColor=<rgba 0.7109 0 1 1>}
  n6 = VectorParameter() {group=Color Customization, parameterName=Base Color, sortPriority=30, defaultValue=<rgba 1 1 1 1>}
  n7 = VectorParameter() {group=Color Customization, parameterName=Base Blend Color, sortPriority=31, defaultValue=<rgba 1 1 1 1>}
  n8 = NamedRerouteUsage()
  NamedRerouteDeclaration(None=StaticSwitch(True=LinearInterpolate(A=@n6.RGB, B=@n7.RGB, Alpha=@n8), False=@n6.RGB, Value=NamedRerouteUsage())) {name=BaseColor, nodeColor=<rgba 0 1 0.3906 1>}
  NamedRerouteDeclaration(None=@n1.Result) {name=IN_Material, nodeColor=<rgba 1 0.0703 0 1>}
  n9 = GetMaterialAttributes(None=NamedRerouteUsage()) {attributeGetTypes [
    #1
    #2
    #3
    #4
    #5
   ]}
  n10 = NamedRerouteDeclaration(None=@n9.Base Color) {name=IN_BaseColor, nodeColor=<rgba 0 0.25 1 1>}
  NamedRerouteDeclaration(None=ChannelMaskParameter(None=@n10) {group=Color Customization, parameterName=BaseColorAdvancedBlendChannel, sortPriority=18}.R) {name=AdvancedBlendMask, nodeColor=<rgba 0.5703 1 0 1>}
  n11 = ChannelMaskParameter(None=@n10) {group=Color Customization, maskChannel=Green, parameterName=BaseColorValueChannel, sortPriority=15, defaultValue=<rgba 0 1 0 0>}
  n12 = Add(A=@n11.R, B=ScalarParameter() {group=Color Customization, parameterName=BaseColorValueBias, sortPriority=16})
  NamedRerouteDeclaration(None=Power(
    Base=Saturate(None=@n12)
    Exp=ScalarParameter() {defaultValue=1, group=Color Customization, parameterName=BaseColorValuePow, sortPriority=17}
   )) {name=BaseColorValue, nodeColor=<rgba 1 0 0.8906 1>}
  n13 = VectorParameter() {group=Color Customization, parameterName=Primary Color, defaultValue=<rgba 1 1 1 1>}
  n14 = VectorParameter() {group=Color Customization, parameterName=Primary Blend Color, sortPriority=33, defaultValue=<rgba 1 1 1 1>}
  n15 = NamedRerouteUsage()
  NamedRerouteDeclaration(None=StaticSwitch(True=LinearInterpolate(A=@n13.RGB, B=@n14.RGB, Alpha=@n15), False=@n13.RGB, Value=NamedRerouteUsage())) {name=PrimaryColor, nodeColor=<rgba 0 1 0.7891 1>}
  n16 = VectorParameter() {group=Color Customization, parameterName=Secondary Color, sortPriority=34, defaultValue=<rgba 1 1 1 1>}
  n17 = VectorParameter() {group=Color Customization, parameterName=Secondary Blend Color, sortPriority=35, defaultValue=<rgba 1 1 1 1>}
  n18 = NamedRerouteUsage()
  NamedRerouteDeclaration(None=StaticSwitch(True=LinearInterpolate(A=@n16.RGB, B=@n17.RGB, Alpha=@n18), False=@n16.RGB, Value=NamedRerouteUsage())) {name=SecondaryColor, nodeColor=<rgba 0 1 0.7891 1>}
  n19 = VectorParameter() {group=Color Customization, parameterName=Tertiary Color, sortPriority=36, defaultValue=<rgba 1 1 1 1>}
  n20 = VectorParameter() {group=Color Customization, parameterName=Tertiary Blend Color, sortPriority=37, defaultValue=<rgba 1 1 1 1>}
  n21 = NamedRerouteUsage()
  NamedRerouteDeclaration(None=StaticSwitch(True=LinearInterpolate(A=@n19.RGB, B=@n20.RGB, Alpha=@n21), False=@n19.RGB, Value=NamedRerouteUsage())) {name=TertiaryColor, nodeColor=<rgba 0 1 0.7891 1>}
  n22 = VectorParameter() {
   desc=Put 1.0 in the channel that that the pattern mask is in and 0.0 in the other channels
   group=Color Customization
   parameterName=Pattern Mask Channel
   sortPriority=52
   defaultValue=<rgba 1 0 0 1>
  }
  n23 = DotProduct(A=@n2.RGB, B=@n22.RGB)
  NamedRerouteDeclaration(None=Multiply(A=Saturate(None=@n23), B=ScalarParameter() {group=Color Customization, parameterName=Pattern Strength, sortPriority=54})) {name=PatternMask, nodeColor=<rgba 0.1719 1 0 1>}
  n24 = NamedRerouteUsage()
  n25 = LinearInterpolate(A=NamedRerouteUsage(), B=NamedRerouteUsage(), Alpha=NamedRerouteUsage()) {constA=1}
  n26 = NamedRerouteUsage()
  n27 = NamedRerouteUsage()
  n28 = LinearInterpolate(
   A=LinearInterpolate(A=LinearInterpolate(A=@n25, B=@n26, Alpha=@n27), B=NamedRerouteUsage(), Alpha=NamedRerouteUsage())
   B=NamedRerouteUsage()
   Alpha=NamedRerouteUsage()
  )
  n29 = VectorParameter() {group=Color Customization, parameterName=Pattern Color, sortPriority=53, defaultValue=<rgba 1 1 1 1>}
  n30 = StaticSwitchParameter(True=LinearInterpolate(A=@n28, B=@n29.RGB, Alpha=NamedRerouteUsage()) {constA=1}, False=@n28) {group=Color Customization, parameterName=Use Pattern Mask, sortPriority=50}
  NamedRerouteDeclaration(None=LinearInterpolate(A=NamedRerouteUsage(), B=Multiply(A=@n24, B=@n30), Alpha=NamedRerouteUsage())) {name=OUT_BaseColor, nodeColor=<rgba 1 0 0.4922 1>}
  NamedRerouteDeclaration(None=ChannelMaskParameter(None=NamedRerouteUsage()) {group=Color Customization, parameterName=Tech Art Mask Channel, sortPriority=10}.R) {name=Mask, nodeColor=<rgba 0 0.8125 1 1>}
  n31 = LocalPosition() {localOrigin=InstancePreSkinning}
  n32 = PreSkinnedNormal()
  n33 = MaterialFunctionCall!(
   Local Position=ShaderStageSwitch(PixelShader=VertexInterpolator(VS=@n31.XYZ).PS, VertexShader=@n31.XYZ)
   Local Normal=ShaderStageSwitch(PixelShader=VertexInterpolator(VS=@n32).PS, VertexShader=@n32)
  ) {materialFunction=/BRCosmetics/Characters/Player/Male/Medium/Bodies/M_MED_AuraCop_Bullet/Materials/MaterialFunctions/MF_AuraCop_GoldMask}
  NamedRerouteDeclaration(None=Saturate(None=ComponentMask(None=@n33.GoldMask) {r=true})) {name=PEV_Mask, nodeColor=<rgba 0 0.0422 1 1>}
  NamedRerouteDeclaration(None=Saturate(None=ComponentMask(None=@n33.GoldMask) {g=true})) {name=PEV_Emissive, nodeColor=<rgba 0 0.1562 1 1>}
  n34 = ComponentMask(None=NamedRerouteUsage()) {g=true}
  n35 = ScalarParameter() {defaultValue=0.3, group=AuraCop, parameterName=GMapIntensity}
  n36 = LinearInterpolate(
   A=Saturate(None=Add(A=@n34, B=@n35))
   B=ScalarParameter() {defaultValue=0.68, group=AuraCop, parameterName=TattooGoldDarkness}
   Alpha=NamedRerouteUsage()
  ) {constA=1, constB=0.68}
  n37 = ComponentMask(None=VertexInterpolator(VS=PreSkinnedNormal()).PS) {b=true}
  n38 = LinearInterpolate(
   A=VectorParameter() {group=AuraCop, parameterName=GoldBaseColor, defaultValue=<rgba 0.5 0.221 0.0183 1>}.RGB
   B=VectorParameter() {group=AuraCop, parameterName=GoldBaseColor_Y_Top, defaultValue=<rgba 0.8073 0.5146 0.1598 1>}.RGB
   Alpha=SmoothStep(
    Min=ScalarParameter() {defaultValue=0.5, group=AuraCop, parameterName=GoldBaseColor_Min}
    Max=ScalarParameter() {defaultValue=1, group=AuraCop, parameterName=GoldBaseColor_Max}
    Value=Saturate(None=@n37)
   ) {constMin=0.5}
  )
  n39 = Constant() {r=1}
  n40 = Constant()
  n41 = SetMaterialAttributes!(
   MaterialAttributes=NamedRerouteUsage()
   Base Color=NamedRerouteDeclaration(None=Multiply(A=@n36, B=@n38)) {name=GoldColor, nodeColor=<rgba 1 0 0.75 1>}
   Specular=@n39
   Roughness=LinearInterpolate(Alpha=NamedRerouteUsage()) {constA=0.25, constB=0.1}
   Metallic=@n39
   Emissive Color=@n40
   Normal=NamedRerouteUsage()
   Opacity=@n40
   Shading Model=ShadingModel()
  ) {attributeSetTypes [
    #1
    #3
    #4
    #2
    #6
    #5
    #7
    #8
   ]}
  n42 = NamedRerouteUsage()
  n43 = Constant()
  n44 = If(
   A=ScalarParameter() {defaultValue=100, group=MidasTech_GoldMask, parameterName=GM_CurrentStage, sortPriority=3}
   "A > B"=Constant() {r=1}
   "A == B"=@n43
   "A < B"=@n43
  )
  n45 = GetMaterialAttributes(None=BlendMaterialAttributes(A=NamedRerouteUsage(), B=@n41, Alpha=Multiply(A=@n42, B=@n44))) {attributeGetTypes [#6, #2]}
  NamedRerouteDeclaration(None=@n45.Emissive Color) {name=CurrentEmissive, nodeColor=<rgba 0.9922 1 0 1>}
  NamedRerouteDeclaration(None=@n45.Metallic) {name=CurrentMetallic, nodeColor=<rgba 0 1 0.6953 1>}
  NamedRerouteDeclaration(None=@n33.TimeNative) {name=TimeNative, nodeColor=<rgba 0.0547 1 0 1>}
  n46 = NamedRerouteUsage()
  n47 = VectorParameter() {group=MidasTech_GoldMask, parameterName=GM_Transition_EmissiveColor, sortPriority=101, defaultValue=<rgba 20 8.8102 1 1>}
  n48 = NamedRerouteUsage()
  n49 = Multiply(A=@n47.RGB, B=SmoothStep(Value=OneMinus(None=@n48)) {constMax=0.2, constValue=1})
  n50 = ScalarParameter() {defaultValue=1, group=MidasTech_GoldMask, parameterName=GM_Transition_EmissiveIntensity}
  n51 = LinearInterpolate(A=@n46, B=Add(A=@n46, B=Multiply(A=@n49, B=@n50)), Alpha=NamedRerouteUsage())
  n52 = OneMinus(None=NamedRerouteUsage())
  NamedRerouteDeclaration(None=SetMaterialAttributes!(MaterialAttributes=@n45.MaterialAttributes, Emissive Color=Multiply(A=@n51, B=@n52)) {attributeSetTypes [#6]}) {name=GoldEffectMaterial, nodeColor=<rgba 0 0.5078 1 1>}
  NamedRerouteDeclaration(None=@n9.Metallic) {name=IN_Metallic, nodeColor=<rgba 0.8047 0 1 1>}
  NamedRerouteDeclaration(None=@n9.Specular) {name=IN_Spec, nodeColor=<rgba 0 1 0.4844 1>}
  NamedRerouteDeclaration(None=@n9.Roughness) {name=IN_Roughness, nodeColor=<rgba 1 0 0.8203 1>}
  n53 = NamedRerouteUsage()
  n54 = ChannelMaskParameter(None=@n53) {group=Color Customization, maskChannel=Green, parameterName=Constant Gold Channel, sortPriority=10, defaultValue=<rgba 0 1 0 0>}
  n55 = ChannelMaskParameter(None=@n53) {group=Color Customization, parameterName=Gold Scar Channel, sortPriority=10}
  n56 = ScalarParameter() {controlType=Enumeration, enumeration=/Game/Global/Materials/Enums/E_Mat_EnabledDisabled, parameterName=ScarEye}
  NamedRerouteDeclaration(None=Add(A=@n54.R, B=LinearInterpolate(B=@n55.R, Alpha=@n56))) {name=ConstantGoldMask, nodeColor=<rgba 1 0.2109 0 1>}
  n57 = NamedRerouteUsage()
  n58 = ScalarParameter() {defaultValue=0.08, parameterName=FlattenGoldThreshold}
  n59 = ScalarParameter() {defaultValue=0.03, parameterName=FlattenGoldSoftness}
  n60 = Subtract(A=@n58, B=@n59)
  n61 = Add(A=@n58, B=@n59)
  n62 = Length(None=ComponentMask(None=@n57) {g=true, r=true}) {is=class MaterialExpressionLength}
  NamedRerouteDeclaration(None=LinearInterpolate(A=Constant3Vector() {constant=<rgba 0 0 1 1>}, B=@n57, Alpha=SmoothStep(Min=@n60, Max=@n61, Value=@n62))) {name=GoldNormal, nodeColor=<rgba 1 0.8906 0 1>}
  NamedRerouteDeclaration(None=@n9.Normal) {name=IN_Normal, nodeColor=<rgba 0.8984 0 1 1>}
  n63 = SetMaterialAttributes!(
   MaterialAttributes=NamedRerouteUsage()
   Base Color=LinearInterpolate(A=NamedRerouteUsage(), B=NamedRerouteUsage(), Alpha=NamedRerouteUsage())
   Metallic=LinearInterpolate(A=NamedRerouteUsage(), Alpha=NamedRerouteUsage())
   Specular=LinearInterpolate(A=NamedRerouteUsage(), Alpha=NamedRerouteUsage())
   Roughness=LinearInterpolate(A=NamedRerouteUsage(), Alpha=NamedRerouteUsage()) {constB=0.25}
   Normal=LinearInterpolate(A=NamedRerouteUsage(), B=NamedRerouteUsage(), Alpha=NamedRerouteUsage())
  ) {attributeSetTypes [
    #1
    #2
    #3
    #4
    #5
   ]}
  NamedRerouteDeclaration(None=StaticSwitchParameter(True=@n63, False=NamedRerouteUsage()) {group=Color Customization, parameterName=Use Color Customization, sortPriority=0}) {name=CustomColorMaterial, nodeColor=<rgba 1 0.7266 0 1>}
  NamedRerouteDeclaration(None=@n1.Tech Art Mask) {name=TechArtMask, nodeColor=<rgba 1 0.1875 0 1>}
  NamedRerouteDeclaration(None=ChannelMaskParameter(None=@n4.RGB) {group=Color Customization, parameterName=PrimaryMaskChannel, sortPriority=12}.R) {name=PrimaryMask, nodeColor=<rgba 1 0.0596 0.1687 1>}
  n64 = ChannelMaskParameter(None=@n4.RGB) {group=Color Customization, maskChannel=Green, parameterName=SecondaryMaskChannel, sortPriority=13, defaultValue=<rgba 0 1 0 0>}
  NamedRerouteDeclaration(None=@n64.R) {name=SecondaryMask, nodeColor=<rgba 0 1 0.0283 1>}
 ]
}
